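# Mono MetaSound Source that plays the looping ambient sound for the CherrySmoke chest.
# A thin wrapper around the shared MSS_PlayRandomSoundLooping_Mono graph, it passes through a one-sound array, attack and release times, and slight random volume variation with fixed pitch.
# It is spatialized with the CherrySmoke chest attenuation, routed to the loot ambient loop sound class, and limited by an ambient chest concurrency set.
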MetaSoundSource /CherrySmokeChestParent/Assets/Audio/MSS_CherrySmokeChest_Amb_loop
inputs [
 Attack Time {type=Time}
 bUseRandomStartTime {type=Bool}
 Loop {type=Bool}
 Max Pitch {type=Float}
 Max Volume {type=Float}
 Min Pitch {type=Float}
 Min Volume {type=Float}
 Release Time {type=Time}
 Sounds {type=WaveAsset:Array}
 UE.Source.OnPlay {type=Trigger}
]
interfaces [UE.OutputFormat.Mono:1.0, UE.Source:1.1]
outputs ["UE.OutputFormat.Mono.Audio:0" {type=Audio}]
pages [Default {
  input_defaults {
   Attack Time=0.2
   Loop="<MetaSoundBuilderResult.SUCCEEDED: 0>"
   Max Pitch=1
   Max Volume=1
   Min Pitch=1
   Min Volume=0.95
   Release Time=0.2
   UE.Source.OnPlay=null
   bUseRandomStartTime=null
   Sounds [&/CherrySmokeChest_Ambient_Loop_01]
  }
  nodes [MSS_PlayRandomSoundLooping_Mono_0 = /FortMetasound/Sources/MSS_PlayRandomSoundLooping_Mono(
    Attack Time=^Attack Time
    Loop=^Loop
    Max Pitch=^Max Pitch
    Max Volume=^Max Volume
    Min Pitch=^Min Pitch
    Min Volume=^Min Volume
    Release Time=^Release Time
    Sounds=^Sounds
    UE.Source.OnPlay=^UE.Source.OnPlay
    bUseRandomStartTime=^bUseRandomStartTime
   )]
  links [{dst="@UE.OutputFormat.Mono.Audio:0", src=MSS_PlayRandomSoundLooping_Mono_0.UE.OutputFormat.Mono.Audio:0}]
 }]
properties {
 assetImportData=&self:AssetImportData
 attenuationSettings=&/ATTN_Ch6S3_CherrySmoke_Chests
 bLooping=true
 duration=10000
 maxDistance=1250
 sampleRate=48000
 soundClassObject=/Game/Sounds/FortSoundClassesAndModes/Fort_Loot_Ambient_Loop
 concurrencyOverrides {maxCount=3}
 concurrencySet [&/Conc_CherrySmokeChest_Amb]
}
