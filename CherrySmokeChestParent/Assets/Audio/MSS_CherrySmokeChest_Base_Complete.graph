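# Mono one-shot MetaSound source for the Cherry Smoke Chest "complete" sound.
# Its inputs feed a nested MSS_PlayRandomSound_Mono, which supplies the audio output and finished trigger from a sound array with pitch and volume ranges.
# The default array holds a single wave, and Master Volume drives the nested Volume input.
# Fortnite item-interact attenuation, the loot reward sound class and a prevent-new concurrency override govern playback.
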MetaSoundSource /CherrySmokeChestParent/Assets/Audio/MSS_CherrySmokeChest_Base_Complete
inputs [
 Master Volume {type=Float}
 MaxPitch {type=Float}
 MaxVolume {type=Float}
 MinPitch {type=Float}
 MinVolume {type=Float}
 Sounds {type=WaveAsset:Array}
 UE.Source.OnPlay {type=Trigger}
]
interfaces [UE.OutputFormat.Mono:1.0, UE.Source.OneShot:1.0, UE.Source:1.1]
outputs [
 "UE.OutputFormat.Mono.Audio:0" {type=Audio}
 UE.Source.OneShot.OnFinished {type=Trigger}
]
pages [Default {
  input_defaults {
   Master Volume=1
   MaxPitch=1.05
   MaxVolume=0.8
   MinPitch=0.8
   MinVolume=0.9
   UE.Source.OnPlay=null
   Sounds [&/CherrySmokeChest_Complete_01]
  }
  nodes [MSS_PlayRandomSound_Mono_0 = &MSS_PlayRandomSound_Mono(
    Volume=^Master Volume
    MaxPitch=^MaxPitch
    MaxVolume=^MaxVolume
    MinPitch=^MinPitch
    MinVolume=^MinVolume
    Sounds=^Sounds
    UE.Source.OnPlay=^UE.Source.OnPlay
   )]
  links [
   {dst="@UE.OutputFormat.Mono.Audio:0", src=MSS_PlayRandomSound_Mono_0.UE.OutputFormat.Mono.Audio:0}
   {dst="@UE.Source.OneShot.OnFinished", src=MSS_PlayRandomSound_Mono_0.UE.Source.OneShot.OnFinished}
  ]
 }]
properties {
 assetImportData=&self:AssetImportData
 attenuationSettings=/Game/Sounds/Attenuation/Fort_Item_Interact_Mono_Att
 bOverrideConcurrency=true
 maxDistance=1500
 sampleRate=48000
 soundClassObject=&Fort_Loot_Reward_PSM
 volume=1.2
 concurrencyOverrides {resolutionRule=PreventNew}
}
