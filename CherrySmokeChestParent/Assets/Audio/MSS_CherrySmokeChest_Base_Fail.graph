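# One-shot mono MetaSound source for the Cherry Smoke Chest fail sound, using item-interact attenuation and the loot reward sound class.
# On play, a single Wave Player plays the exposed wave input, pitch-shifted by the shared MSP_RandomPitchTimeVolume patch, and signals finished when playback ends.
# Only the patch's random pitch output is used.
# A concurrency override prevents new instances from starting once the limit is reached.
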MetaSoundSource /CherrySmokeChestParent/Assets/Audio/MSS_CherrySmokeChest_Base_Fail
inputs [
 "Random Pitch Max (Semitones)" {type=Float}
 Sounds {type=WaveAsset}
 UE.Source.OnPlay {type=Trigger}
]
interfaces [UE.OutputFormat.Mono:1.0, UE.Source.OneShot:1.0, UE.Source:1.1]
outputs [
 "UE.OutputFormat.Mono.Audio:0" {type=Audio}
 UE.Source.OneShot.OnFinished {type=Trigger}
]
pages [Default {
  input_defaults {"Random Pitch Max (Semitones)"=1, Sounds=&/CherrySmokeChest_Fail_01, UE.Source.OnPlay=null}
  nodes [WavePlayer_0 = UE.Wave Player:Mono(
    Wave Asset=^Sounds
    Play=^UE.Source.OnPlay
    Pitch Shift=/FortMetasound/Patches/MSP_RandomPitchTimeVolume("Random Pitch Max (Semitones)"=^"Random Pitch Max (Semitones)") {also {defaults {"Random Pitch Min (Semitones)"=0.8, "Random Volume Change Min (Decibels)"=0.95}}}.Random Pitch
   )]
  links [{dst="@UE.Source.OneShot.OnFinished", src=WavePlayer_0.On Finished}, {dst="@UE.OutputFormat.Mono.Audio:0", src=WavePlayer_0.Out Mono}]
 }]
properties {
 assetImportData=&self:AssetImportData
 attenuationSettings=/Game/Sounds/Attenuation/Fort_Item_Interact_Mono_Att
 bOverrideConcurrency=true
 maxDistance=1500
 sampleRate=48000
 soundClassObject=&Fort_Loot_Reward_PSM
 volume=1.2
 concurrencyOverrides {resolutionRule=PreventNew}
}
