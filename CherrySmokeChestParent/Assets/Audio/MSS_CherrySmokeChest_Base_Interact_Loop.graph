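# MetaSound source for the looping mono interact sound of the Cherry Smoke chest, presumably heard while a player interacts with it.
# It wraps the shared MSS_PlayRandomSoundLooping_Mono graph, passing through its pitch, volume, attack, release and loop inputs and routing its output straight out.
# The sound array holds a single interact loop wave, so variation comes from the pitch and volume ranges.
# Playback uses item-interact mono attenuation and the loot reward sound class.
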MetaSoundSource /CherrySmokeChestParent/Assets/Audio/MSS_CherrySmokeChest_Base_Interact_Loop
inputs [
 Attack Time {type=Time}
 bUseRandomStartTime {type=Bool}
 Loop {type=Bool}
 Max Pitch {type=Float}
 Max Volume {type=Float}
 Min Pitch {type=Float}
 Min Volume {type=Float}
 Release Time {type=Time}
 Sounds {type=WaveAsset:Array}
 UE.Source.OnPlay {type=Trigger}
]
interfaces [UE.OutputFormat.Mono:1.0, UE.Source:1.1]
outputs ["UE.OutputFormat.Mono.Audio:0" {type=Audio}]
pages [Default {
  input_defaults {
   Attack Time=0.2
   Loop="<MetaSoundBuilderResult.SUCCEEDED: 0>"
   Max Pitch=1.5
   Max Volume=1
   Min Pitch=0.95
   Min Volume=0.9
   Release Time=0.2
   UE.Source.OnPlay=null
   bUseRandomStartTime=null
   Sounds [&/CherrySmokeChest_Interact_Loop_01]
  }
  nodes [MSS_PlayRandomSoundLooping_Mono_0 = /FortMetasound/Sources/MSS_PlayRandomSoundLooping_Mono(
    Attack Time=^Attack Time
    Loop=^Loop
    Max Pitch=^Max Pitch
    Max Volume=^Max Volume
    Min Pitch=^Min Pitch
    Min Volume=^Min Volume
    Release Time=^Release Time
    Sounds=^Sounds
    UE.Source.OnPlay=^UE.Source.OnPlay
    bUseRandomStartTime=^bUseRandomStartTime
   )]
  links [{dst="@UE.OutputFormat.Mono.Audio:0", src=MSS_PlayRandomSoundLooping_Mono_0.UE.OutputFormat.Mono.Audio:0}]
 }]
properties {
 assetImportData=&self:AssetImportData
 attenuationSettings=/Game/Sounds/Attenuation/Fort_Item_Interact_Mono_Att
 bLooping=true
 duration=10000
 maxDistance=1500
 sampleRate=48000
 soundClassObject=&Fort_Loot_Reward_PSM
 volume=1.3
}
